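# First-person fire sound cue for the Striker Auto SR weapon in Athena, playing a single close-range single-fire sample.
# A Modulator applies slight random pitch and volume variation to the one Wave Player on each shot.
# First-person weapon attenuation, sound class and gunfire concurrency assets are assigned, with a boosted cue volume and owner-limited, stop-oldest concurrency override values.
# The debug flag is left enabled.
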SoundCue /Game/Athena/Sounds/Weapons/Striker_Auto_SR/Weapon_ASLT_AutoSF_SR_1P_Athena_Cue
root=Modulator_6
nodes [
 {id=Modulator_6, children [WavePlayer_4], properties {pitchMax=1.025, pitchMin=0.98, volumeMax=1, volumeMin=0.9}}
 {id=WavePlayer_4, properties {soundWaveAssetPtr=/Game/Athena/Sounds/Weapons/Striker_Auto_SR/SingleFire/sfx_strikerSR_singlefire_close_01}}
]
properties {
 attenuationSettings=&Fort_Weapon_Projectile_FirstPerson_Athena_Att
 bDebug=true
 maxDistance=2097152
 soundClassObject=&Fort_WeaponRanged_Fire_1P_PSM
 volumeMultiplier=1.6
 concurrencyOverrides {bLimitToOwner=true, maxCount=3, resolutionRule=StopOldest}
 concurrencySet [&Weapon_Gunfire_FirstPerson_Conc]
}
