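# Sound cue for Striker Auto SR single-fire gunshots in Athena, classed as third-person ranged weapon fire.
# Nested mixers blend close, medium and distant layers, each with its own attenuation, with the medium and distant layers overridden to the distant-fire sound class.
# The medium layer randomly picks among three waves, and modulators add slight pitch variation to each layer.
# Global and local gunfire concurrency sets are assigned.
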SoundCue /Game/Athena/Sounds/Weapons/Striker_Auto_SR/Weapon_ASLT_AutoSF_SR_Athena_Cue
root=Mixer_4
nodes [
 {id=Attenuation_6, children [Modulator_6], properties {attenuationSettings=&Fort_Weapon_Projectile_Close_Athena_Att}}
 {
  id=Attenuation_7
  children [Modulator_4]
  properties {attenuationSettings=&Fort_Weapon_Projectile_Distant_Mono_Athena_Att}
 }
 {
  id=Attenuation_8
  children [Modulator_7]
  properties {attenuationSettings=&Fort_Weapon_Projectile_Medium_Mono_Athena_Att}
 }
 {id=Mixer_4, children [Mixer_5,SoundClass_8], properties {inputVolume [1,0.4]}}
 {id=Mixer_5, children [SoundClass_7,SoundClass_9], properties {inputVolume [1,2]}}
 {id=Modulator_4, children [WavePlayer_2], properties {pitchMax=1.02, pitchMin=0.98, volumeMax=2, volumeMin=2}}
 {id=Modulator_6, children [WavePlayer_4], properties {pitchMin=0.98, volumeMax=1, volumeMin=0.9}}
 {id=Modulator_7, children [Random_1], properties {pitchMax=1, pitchMin=0.98, volumeMax=1, volumeMin=1}}
 {id=Random_1, children [WavePlayer_6,WavePlayer_7,WavePlayer_8], properties {bShouldExcludeFromBranchCulling=true, weights [1,1,1]}}
 {
  id=SoundClass_7
  children [Attenuation_6]
  properties {soundClassOverride=&Fort_WeaponRanged_Fire_3P_PSM}
 }
 {
  id=SoundClass_8
  children [Attenuation_7]
  properties {soundClassOverride=&Fort_WeaponRanged_Fire_Distant}
 }
 {
  id=SoundClass_9
  children [Attenuation_8]
  properties {soundClassOverride=&Fort_WeaponRanged_Fire_Distant}
 }
 {id=WavePlayer_2, properties {soundWaveAssetPtr=/Game/Athena/Sounds/Weapons/Striker_Auto_SR/SingleFire/sfx_strikerSR_singlefire_distantNew}}
 {id=WavePlayer_4, properties {soundWaveAssetPtr=/Game/Athena/Sounds/Weapons/Striker_Auto_SR/SingleFire/sfx_strikerSR_singlefire_close_01}}
 {id=WavePlayer_6, properties {soundWaveAssetPtr=/Game/Athena/Sounds/Weapons/Striker_Auto_SR/SingleFire/sfx_strikerSR_singlefire_mediumNew_01}}
 {id=WavePlayer_7, properties {soundWaveAssetPtr=/Game/Athena/Sounds/Weapons/Striker_Auto_SR/SingleFire/sfx_strikerSR_singlefire_mediumNew_02}}
 {id=WavePlayer_8, properties {soundWaveAssetPtr=/Game/Athena/Sounds/Weapons/Striker_Auto_SR/SingleFire/sfx_strikerSR_singlefire_mediumNew_03}}
]
properties {
 bDebug=true
 maxDistance=27500
 soundClassObject=&Fort_WeaponRanged_Fire_3P_PSM
 volumeMultiplier=1
 concurrencyOverrides {bLimitToOwner=true, maxCount=3, resolutionRule=StopOldest}
 concurrencySet [&Weapon_Gunfire_Global_Conc, &Weapon_Gunfire_Local_Conc]
}
